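# Sound cue for leaving aim-down-sights with the Scrap bow.
# A single wave player reuses the heavy assault rifle ADS-out sample, pitched down and quieted at cue level, through a Modulator left at default settings.
# It uses weapon reload attenuation and the weapon foley sound class, with both global and local bow concurrency rules.
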SoundCue /PrimalGameplay/Sounds/Weapons/Bows/Scrap/Bow_Scrap_ADS_Out_Cue
root=Modulator_0
nodes [
 {id=Modulator_0, children [WavePlayer_0]}
 {id=WavePlayer_0, properties {soundWaveAssetPtr=/Game/Sounds/Fort_Weapons/ADS/Weapon_HEAVY_AR_ADS_OUT_01}}
]
properties {
 attenuationSettings=&Fort_Weapon_Reload_Att
 maxDistance=2250
 pitchMultiplier=0.8
 soundClassObject=&Fort_WeaponFoley
 volumeMultiplier=0.376
 concurrencySet [/Game/Sounds/Concurrency_Rules/Bow_Global_Conc, /Game/Sounds/Concurrency_Rules/Bow_Local_Conc]
}
