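# Sound cue for equipping the Scrap bow, played as weapon foley.
# A Random node picks one of two equip wave variations with equal weight.
# It uses weapon reload attenuation, routes through the weapon foley sound class at reduced volume, and is limited by shared global and local bow concurrency rules.
# The cue has its debug flag enabled.
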SoundCue /PrimalGameplay/Sounds/Weapons/Bows/Scrap/Bow_Scrap_Equip_Cue
root=Random_0
nodes [
 {id=Random_0, children [WavePlayer_1,WavePlayer_2], properties {weights [1,1]}}
 {id=WavePlayer_1, properties {soundWaveAssetPtr=&/Bow_Scrap_Equip_01}}
 {id=WavePlayer_2, properties {soundWaveAssetPtr=&/Bow_Scrap_Equip_02}}
]
properties {
 attenuationSettings=&Fort_Weapon_Reload_Att
 bDebug=true
 maxDistance=2250
 soundClassObject=&Fort_WeaponFoley
 volumeMultiplier=0.601
 concurrencySet [/Game/Sounds/Concurrency_Rules/Bow_Global_Conc, /Game/Sounds/Concurrency_Rules/Bow_Local_Conc]
}
