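# First-person sound cue for firing the Scrap bow at low tension, routed to the ranged-weapon 1P fire sound class.
# A Mixer layers a randomly chosen scrap fire sample with a quieter, randomly chosen distant metal bow passby borrowed from the Metal bow set.
# Each layer runs through its own Modulator for variation, with first-person bow projectile attenuation and gunfire concurrency applied to the cue.
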SoundCue /PrimalGameplay/Sounds/Weapons/Bows/Scrap/Bow_Scrap_Fire_Low_Tension_1P_Cue
root=Mixer_1
nodes [
 {id=Mixer_1, children [Modulator_1,Modulator_2], properties {inputVolume [1,0.3]}}
 {id=Modulator_1, children [Random_2]}
 {id=Modulator_2, children [Random_1]}
 {id=Random_1, children [WavePlayer_3,WavePlayer_4,WavePlayer_5,WavePlayer_6,WavePlayer_7,WavePlayer_8], properties {weights [1,1,1,1,1,1]}}
 {id=Random_2, children [WavePlayer_9, WavePlayer_11], properties {weights [1,1]}}
 {id=WavePlayer_11, properties {soundWaveAssetPtr=&/Bow_Scrap_Fire_Low_Tension_03}}
 {id=WavePlayer_3, properties {soundWaveAssetPtr=/PrimalGameplay/Sounds/Weapons/Bows/Metal/Bow_Metal_Passby_Default_Distant_06}}
 {id=WavePlayer_4, properties {soundWaveAssetPtr=/PrimalGameplay/Sounds/Weapons/Bows/Metal/Bow_Metal_Passby_Default_Distant_01}}
 {id=WavePlayer_5, properties {soundWaveAssetPtr=/PrimalGameplay/Sounds/Weapons/Bows/Metal/Bow_Metal_Passby_Default_Distant_02}}
 {id=WavePlayer_6, properties {soundWaveAssetPtr=/PrimalGameplay/Sounds/Weapons/Bows/Metal/Bow_Metal_Passby_Default_Distant_03}}
 {id=WavePlayer_7, properties {soundWaveAssetPtr=/PrimalGameplay/Sounds/Weapons/Bows/Metal/Bow_Metal_Passby_Default_Distant_04}}
 {id=WavePlayer_8, properties {soundWaveAssetPtr=/PrimalGameplay/Sounds/Weapons/Bows/Metal/Bow_Metal_Passby_Default_Distant_05}}
 {id=WavePlayer_9, properties {soundWaveAssetPtr=&/Bow_Scrap_Fire_Low_Tension_01}}
]
properties {
 attenuationSettings=/Game/Sounds/Attenuation/Fort_Weapon_Projectile_FirstPerson_Bow_Athena_Att
 maxDistance=2097152
 soundClassObject=&Fort_WeaponRanged_Fire_1P_PSM
 volumeMultiplier=0.7
 attenuationOverrides {falloffDistance=5000, attenuationShapeExtents {x=1000}}
 concurrencySet [&Weapon_Gunfire_FirstPerson_Conc]
}
