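# Sound cue for firing the Scrap bow at low tension, using third-person and distant ranged-weapon sound classes.
# The close layer mixes randomized, modulated Scrap low-tension shots with quieter randomized Metal bow passby sounds under close attenuation.
# A slightly quieter distant layer picks among borrowed Metal bow distant fire waves under its own attenuation and sound class.
# Global and local bow concurrency rules limit overlapping playback.
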SoundCue /PrimalGameplay/Sounds/Weapons/Bows/Scrap/Bow_Scrap_Fire_Low_Tension_Cue
root=Mixer_3
nodes [
 {id=Attenuation_0, children [Mixer_1], properties {attenuationSettings=/Game/Sounds/Attenuation/Fort_Weapon_Projectile_Close_Bow_Athena_Att}}
 {
  id=Attenuation_2
  children [Random_0]
  properties {attenuationSettings=/Game/Sounds/Attenuation/Fort_Weapon_Projectile_Distant_Bow_Athena_Att}
 }
 {id=Mixer_1, children [Modulator_1,Modulator_0], properties {inputVolume [1,0.3]}}
 {id=Mixer_3, children [SoundClass_0,SoundClass_4], properties {inputVolume [1,0.7]}}
 {id=Modulator_0, children [Random_1]}
 {id=Modulator_1, children [Random_2]}
 {id=Random_0, children [WavePlayer_0,WavePlayer_1,WavePlayer_2], properties {weights [1,1,1]}}
 {id=Random_1, children [WavePlayer_3,WavePlayer_4,WavePlayer_5,WavePlayer_6,WavePlayer_7,WavePlayer_8], properties {weights [1,1,1,1,1,1]}}
 {id=Random_2, children [WavePlayer_9, WavePlayer_19], properties {weights [1,1]}}
 {
  id=SoundClass_0
  children [Attenuation_0]
  properties {soundClassOverride=&Fort_WeaponRanged_Fire_3P_PSM}
 }
 {
  id=SoundClass_4
  children [Attenuation_2]
  properties {soundClassOverride=&Fort_WeaponRanged_Fire_Distant}
 }
 {id=WavePlayer_0, properties {soundWaveAssetPtr=/PrimalGameplay/Sounds/Weapons/Bows/Metal/Bow_Default_Fire_3P_Normal_Distant_01}}
 {id=WavePlayer_1, properties {soundWaveAssetPtr=/PrimalGameplay/Sounds/Weapons/Bows/Metal/Bow_Default_Fire_3P_Normal_Distant_02}}
 {id=WavePlayer_19, properties {soundWaveAssetPtr=&/Bow_Scrap_Fire_Low_Tension_03}}
 {id=WavePlayer_2, properties {soundWaveAssetPtr=/PrimalGameplay/Sounds/Weapons/Bows/Metal/Bow_Default_Fire_3P_Normal_Distant_03}}
 {id=WavePlayer_3, properties {soundWaveAssetPtr=/PrimalGameplay/Sounds/Weapons/Bows/Metal/Bow_Metal_Passby_Default_Distant_06}}
 {id=WavePlayer_4, properties {soundWaveAssetPtr=/PrimalGameplay/Sounds/Weapons/Bows/Metal/Bow_Metal_Passby_Default_Distant_01}}
 {id=WavePlayer_5, properties {soundWaveAssetPtr=/PrimalGameplay/Sounds/Weapons/Bows/Metal/Bow_Metal_Passby_Default_Distant_02}}
 {id=WavePlayer_6, properties {soundWaveAssetPtr=/PrimalGameplay/Sounds/Weapons/Bows/Metal/Bow_Metal_Passby_Default_Distant_03}}
 {id=WavePlayer_7, properties {soundWaveAssetPtr=/PrimalGameplay/Sounds/Weapons/Bows/Metal/Bow_Metal_Passby_Default_Distant_04}}
 {id=WavePlayer_8, properties {soundWaveAssetPtr=/PrimalGameplay/Sounds/Weapons/Bows/Metal/Bow_Metal_Passby_Default_Distant_05}}
 {id=WavePlayer_9, properties {soundWaveAssetPtr=&/Bow_Scrap_Fire_Low_Tension_01}}
]
properties {
 maxDistance=27500
 volumeMultiplier=0.7
 attenuationOverrides {falloffDistance=5000, attenuationShapeExtents {x=1000}}
 concurrencySet [/Game/Sounds/Concurrency_Rules/Bow_Global_Conc, /Game/Sounds/Concurrency_Rules/Bow_Local_Conc]
}
