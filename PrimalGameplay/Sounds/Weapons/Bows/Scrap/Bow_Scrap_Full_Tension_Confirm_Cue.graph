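# Sound cue for the Scrap Bow's full-tension confirm sound, by its name the feedback for a fully drawn bow.
# A single wave player feeds one confirm sample straight to the output, with lowered volume and pitch and no randomization.
# It uses weapon foley sound class, borrows reload attenuation, and is limited by global and local bow concurrency rules.
# The cue's debug flag is left on.
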SoundCue /PrimalGameplay/Sounds/Weapons/Bows/Scrap/Bow_Scrap_Full_Tension_Confirm_Cue
root=WavePlayer_7
nodes [{id=WavePlayer_7, properties {soundWaveAssetPtr=&/Bow_Scrap_Fire_Full_Tension_Confirm_01}}]
properties {
 attenuationSettings=&Fort_Weapon_Reload_Att
 bDebug=true
 maxDistance=2250
 pitchMultiplier=0.9
 soundClassObject=&Fort_WeaponFoley
 volumeMultiplier=0.601
 concurrencySet [/Game/Sounds/Concurrency_Rules/Bow_Global_Conc, /Game/Sounds/Concurrency_Rules/Bow_Local_Conc]
}
