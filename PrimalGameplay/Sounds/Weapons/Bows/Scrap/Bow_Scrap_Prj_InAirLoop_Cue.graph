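# Looping in-air sound cue for the scrap bow's projectile, reusing the metal bow's in-air loop wave.
# The looping wave runs through a default Modulator into an Enveloper that ramps volume up from a quiet start and shapes pitch over time.
# It uses the arrow in-air attenuation, the ranged-weapon projectile sound class, and both global and local bow concurrency rules.
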SoundCue /PrimalGameplay/Sounds/Weapons/Bows/Scrap/Bow_Scrap_Prj_InAirLoop_Cue
root=Enveloper_0
nodes [
 {
  id=Enveloper_0
  children [Modulator_0]
  properties {
   pitchCurve {editorCurveData {keys [{arriveTangent=1.8801, interpMode=RCIM_Cubic, leaveTangent=1.8801, tangentMode=RCTM_User, time=0.0365, value=0.75}, {time=2}]}}
   volumeCurve {editorCurveData {keys [
      <at 0 0.25 RCIM_Cubic>
      {arriveTangent=2.6958, interpMode=RCIM_Cubic, leaveTangent=2.6958, tangentMode=RCTM_User, time=0.7088, value=0.7816}
      <at 1 1>
     ]}}
  }
 }
 {id=Modulator_0, children [WavePlayer_1]}
 {id=WavePlayer_1, properties {bLooping=true, soundWaveAssetPtr=/PrimalGameplay/Sounds/Weapons/Bows/Metal/Bow_Metal_Prj_InAir_Loop}}
]
properties {
 attenuationSettings=/Game/Sounds/Attenuation/Fort_Weapon_Projectile_Arrow_InAirLoop_Att
 maxDistance=16800
 soundClassObject=&Fort_WeaponRanged_Projectile_Effects
 volumeMultiplier=1
 attenuationOverrides {falloffDistance=5000, reverbWetLevelMax=0.7, reverbWetLevelMin=0.2, attenuationShapeExtents {x=1000}}
 concurrencySet [/Game/Sounds/Concurrency_Rules/Bow_Global_Conc, /Game/Sounds/Concurrency_Rules/Bow_Local_Conc]
}
